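# Conversation database for the Biker NPC in the NPC library, giving the greeting shown when a player starts talking to it.
# The entry point leads to one Speech node with a default line and contextual lines gated by first-ever meeting, per-NPC match conversation count, or specific player outfits.
# Priority curves rank first-meeting and outfit reactions.
# The Speech node then branches through link nodes to shared quest, basic service and Biker dialogue entries.
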ConversationDatabase /NPCLibrary/NPCs/Biker/NPCLibrary_Conversation_Biker
entry_points [{tag=Conversation.Entry.Biker, destinations [#1]}]
nodes [
 ConversationEntryPointNode_7 {
  id=#1
  class=ConversationEntryPointNode
  outputs [#2]
  properties {entryTag {tagName=Conversation.Entry.Biker}}
 }
 Speech_C_4 {
  id=#2
  class=Speech_C
  outputs [#3, #4, #5]
  properties {
   nodeName=""
   generalConfig {
    defaultMessage="Out here, the road's all that matters."
    contextualMessages [
     {
      message=The road's the only home I need.
      contextRequirements [{requirement {class=FortControllerRequirement_IsFirstEverConversationWithNPC}}]
      priority {curve {rowName=Tandem.Speech.Priority.FirstEverInteraction}}
     }
     {
      message="Out here, the road's all that matters."
      contextRequirements [{requirement {class=FortControllerRequirement_CheckNPCMatchConversationCount, properties {nPCUniqueIdTag {tagName=AISpawnerData.Type.Tandem.Biker}}}}]
      priority {curve {rowName=None}}
     }
     {
      message=You look like a good fit for the Riders. Interested?
      contextRequirements [{requirement {class=FortControllerRequirement_CheckNPCMatchConversationCount, properties {nPCUniqueIdTag {tagName=AISpawnerData.Type.Tandem.Biker}}}}]
      priority {curve {rowName=None}}
     }
     {
      message=Don't let the wastelands claim you.
      contextRequirements [{requirement {class=FortControllerRequirement_CheckNPCMatchConversationCount, properties {nPCUniqueIdTag {tagName=AISpawnerData.Type.Tandem.Biker}}}}]
      priority {curve {rowName=None}}
     }
     {
      message=Ride fast. Ride Hard. Live free.
      contextRequirements [{requirement {class=FortControllerRequirement_CheckNPCMatchConversationCount, properties {nPCUniqueIdTag {tagName=AISpawnerData.Type.Tandem.Biker}}}}]
      priority {curve {rowName=None}}
     }
     {
      message=Never was one for that noise you call music.
      contextRequirements [{requirement {
         class=FortControllerRequirement_HasCID
         properties {allowedCIds [/BRCosmetics/Athena/Items/Cosmetics/Characters/CID_097_Athena_Commando_F_RockerPunk]}
        }}]
      priority {curve {rowName=Tandem.Speech.Priority.OutfitReaction}}
     }
     {
      message="How many times do I need to tell you, it's Riders, not Rabbles."
      contextRequirements [{requirement {
         class=FortControllerRequirement_HasCID
         properties {allowedCIds [/BRCosmetics/Athena/Items/Cosmetics/Characters/CID_262_Athena_Commando_M_MadCommander]}
        }}]
      priority {curve {rowName=Tandem.Speech.Priority.OutfitReaction}}
     }
     {
      message=Another noise maker? Hope you can ride.
      contextRequirements [{requirement {
         class=FortControllerRequirement_HasCID
         properties {allowedCIds [/BRCosmetics/Athena/Items/Cosmetics/Characters/CID_268_Athena_Commando_M_RockerPunk]}
        }}]
      priority {curve {rowName=Tandem.Speech.Priority.OutfitReaction}}
     }
     {
      message=You might be just the leader we're looking for...
      contextRequirements [{requirement {
         class=FortControllerRequirement_HasCID
         properties {allowedCIds [/BRCosmetics/Athena/Items/Cosmetics/Characters/CID_A_097_Athena_Commando_F_WastelandWarrior]}
        }}]
      priority {curve {rowName=Tandem.Speech.Priority.OutfitReaction}}
     }
    ]
   }
  }
 }
 ConversationLinkNode_2 {id=#5, class=ConversationLinkNode, properties {remoteEntryTag {tagName=Conversation.Entry.Biker.Dialogue}}}
 ConversationLinkNode_0 {id=#4, class=ConversationLinkNode, properties {remoteEntryTag {tagName=Conversation.Entry.BasicServices}}}
 ConversationLinkNode_1 {id=#3, class=ConversationLinkNode, properties {remoteEntryTag {tagName=Conversation.Entry.NPCLibrary.Quests}}}
]
